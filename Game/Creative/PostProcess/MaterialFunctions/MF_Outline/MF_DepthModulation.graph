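# Material function for the post-process outline that outputs a Fresnel-driven modulation factor, grouped under outline depth variation.
# A Fresnel term with adjustable power is offset, divided by a mask power and saturated into a view-angle mask.
# That mask is scaled by the Fresnel Modulation parameter and offset by a constant to form the output.
# It has no function inputs; three scalar parameters drive it.
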MaterialFunction /Game/Creative/PostProcess/MaterialFunctions/MF_Outline/MF_DepthModulation
graph {nodes [
  n1 = Subtract(A=Fresnel(ExponentIn=ScalarParameter() {defaultValue=20, group=3.2 Outline Depth Variation, parameterName=Fresnel Power}))
  n2 = ScalarParameter() {defaultValue=1, group=3.2 Outline Depth Variation, parameterName=Fresnel Mask Power}
  n3 = Saturate(None=Add(A=Divide(A=@n1, B=@n2)))
  n4 = ScalarParameter() {defaultValue=1.2, group=3.2 Outline Depth Variation, parameterName=Fresnel Modulation}
  FunctionOutput(None=Add(A=Multiply(A=@n3, B=@n4)))
 ]}
