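# MF_Kernel is a post-process material function in the outline set that builds a five-tap, plus-shaped sampling kernel around the current pixel.
# It scales horizontal and vertical unit vectors by a Pixel Width input, defaulting to one, and by scene texel size.
# These offsets are added to and subtracted from the texture coordinate, outputting Center, Left, Right, Top and Bottom UVs.
# It outputs coordinates only and samples no textures itself.
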MaterialFunction /Game/Creative/PostProcess/MaterialFunctions/MF_Outline/MF_Kernel
graph {nodes [
  n1 = TextureCoordinate()
  FunctionOutput(None=@n1) {outputName=Center Pixel}
  n2 = FunctionInput(Preview=Constant() {r=1}) {bUsePreviewValueAsDefault=true, inputName=Pixel Width, inputType=FunctionInput_Scalar}
  n3 = SceneTexelSize()
  n4 = Multiply(A=Multiply(A=Constant2Vector() {r=1}, B=@n2), B=@n3)
  FunctionOutput(None=Add(A=@n1, B=Multiply(A=@n4) {constB=-1})) {outputName=Left Pixel}
  FunctionOutput(None=Add(A=@n1, B=@n4)) {outputName=Right Pixel}
  n5 = Multiply(A=Multiply(A=Constant2Vector() {g=1}, B=@n2), B=@n3)
  FunctionOutput(None=Add(A=@n1, B=Multiply(A=@n5) {constB=-1})) {outputName=Top Pixel}
  FunctionOutput(None=Add(A=@n1, B=@n5)) {outputName=Bottom Pixel}
 ]}
