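# Material function that detects outline edges from scene depth for a post-process outline effect, outputting a binary mask.
# MF_Kernel turns the Thickness input into UVs for a center pixel and four neighbors, whose depths form a Laplacian-style difference with Edge Bias scaling the center.
# Static switches toggle the inside (neighbors minus center) and outside (center minus neighbors) differences.
# The larger difference is stepped against Edge Threshold.
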MaterialFunction /Game/Creative/PostProcess/MaterialFunctions/MF_Outline/MF_DetectEdges_Depth
graph {nodes [
  n1 = Constant() {r=4}
  n2 = MaterialFunctionCall!(Pixel Width=FunctionInput(Preview=Constant() {r=1}) {inputName=Thickness, inputType=FunctionInput_Scalar}) {materialFunction=&/MF_Kernel}
  n3 = Add(
   A=Add(A=SceneDepth(UVs=@n2.Left Pixel).R, B=SceneDepth(UVs=@n2.Right Pixel).R)
   B=Add(A=SceneDepth(UVs=@n2.Top Pixel).R, B=SceneDepth(UVs=@n2.Bottom Pixel).R)
  )
  n4 = Multiply(A=SceneDepth(UVs=@n2.Center Pixel).R, B=ScalarParameter() {defaultValue=3.9, group=3. Outline, parameterName=Edge Bias}) {constB=4}
  n5 = Constant()
  n6 = StaticSwitchParameter(True=Subtract(A=@n3, B=@n4), False=@n5) {defaultValue=true, group=3. Outline, parameterName=useInsideEdge}
  n7 = StaticSwitchParameter(True=Subtract(A=@n4, B=@n3), False=@n5) {defaultValue=true, group=3. Outline, parameterName=useOutsideEdge}
  FunctionOutput(None=Step(Y=FunctionInput(Preview=@n1) {inputName=Edge Threshold, inputType=FunctionInput_Scalar}, X=Max(A=@n6, B=@n7)))
 ]}
