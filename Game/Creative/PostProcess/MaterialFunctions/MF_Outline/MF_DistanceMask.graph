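# Material function for outline post-processing that outputs a binary mask limiting the effect by scene distance.
# MF_Kernel supplies center, top, bottom, left and right pixel UVs from a thickness input, and scene depth is sampled at each.
# The nearest of the five depths is stepped against MaxOutlineDistance and inverted, yielding one within range and zero beyond.
# Using the minimum depth keeps the mask on around silhouettes of near objects against distant backgrounds.
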MaterialFunction /Game/Creative/PostProcess/MaterialFunctions/MF_Outline/MF_DistanceMask
graph {nodes [
  n1 = FunctionInput(Preview=Constant() {r=10000}) {bUsePreviewValueAsDefault=true, inputName=MaxOutlineDistance, inputType=FunctionInput_Scalar}
  n2 = MaterialFunctionCall!(Pixel Width=FunctionInput(Preview=Constant() {r=2}) {bUsePreviewValueAsDefault=true, inputName=Max Thickness, inputType=FunctionInput_Scalar}) {materialFunction=&/MF_Kernel}
  n3 = SceneDepth(UVs=@n2.Top Pixel)
  n4 = Min(A=SceneDepth(UVs=@n2.Bottom Pixel).R, B=SceneDepth(UVs=@n2.Center Pixel).R)
  n5 = Min(A=SceneDepth(UVs=@n2.Left Pixel).R, B=Min(A=SceneDepth(UVs=@n2.Right Pixel).R, B=Min(A=@n3.R, B=@n4)))
  FunctionOutput(None=OneMinus(None=Step(Y=@n1, X=@n5)))
 ]}
