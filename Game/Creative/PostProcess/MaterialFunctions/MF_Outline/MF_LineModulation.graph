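# MF_LineModulation is a post-process outline material function that outputs a line thickness that shrinks with scene depth.
# It feeds Max Thickness into MF_Kernel as pixel width, then samples scene depth at the center, top, bottom, left and right.
# The nearest depth, capped at Max Depth, linearly blends from Max Thickness at zero to Min Thickness at that cap.
# Using the neighborhood minimum makes the width follow the closest nearby surface.
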MaterialFunction /Game/Creative/PostProcess/MaterialFunctions/MF_Outline/MF_LineModulation
graph {nodes [
  n1 = Constant() {r=1}
  n2 = FunctionInput(Preview=Constant() {r=2}) {bUsePreviewValueAsDefault=true, inputName=Max Thickness, inputType=FunctionInput_Scalar}
  n3 = FunctionInput(Preview=Constant() {r=5000}) {bUsePreviewValueAsDefault=true, inputName=Max Depth, inputType=FunctionInput_Scalar}
  n4 = Divide(
   A=Subtract(A=FunctionInput(Preview=@n1) {bUsePreviewValueAsDefault=true, inputName=Min Thickness, inputType=FunctionInput_Scalar}, B=@n2)
   B=@n3
  )
  n5 = MaterialFunctionCall!(Pixel Width=@n2) {materialFunction=&/MF_Kernel}
  n6 = SceneDepth(UVs=@n5.Right Pixel)
  n7 = Min(A=SceneDepth(UVs=@n5.Top Pixel).R, B=Min(A=SceneDepth(UVs=@n5.Bottom Pixel).R, B=SceneDepth(UVs=@n5.Center Pixel).R))
  n8 = Min(A=@n3, B=Min(A=SceneDepth(UVs=@n5.Left Pixel).R, B=Min(A=@n6.R, B=@n7)))
  FunctionOutput(None=Add(A=@n2, B=Multiply(A=@n4, B=@n8)))
 ]}
